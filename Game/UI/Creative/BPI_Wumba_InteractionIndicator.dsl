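; BPI_Wumba_InteractionIndicator is a Blueprint interface in the Creative UI folder for Wumba interaction indicators.
; It declares AttemptWumbaInteraction, which receives the current Fort interact context info and the indicated actor and returns false by default.
; It also declares HandleWumbaInteraction, which receives an integer upgrade direction and returns an upgradable value.
; The file holds only function signatures, leaving behavior to implementing classes.
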(fn AttemptWumbaInteraction
  (params (Current%20Interaction :type Fort%20Interact%20Context%20Info%20Object%20Reference) (Indicated%20Actor :type Actor%20Object%20Reference))
  (return false))

(fn HandleWumbaInteraction
  (params (UpgradeDirection :type Integer))
  (return "Upgradable"))
