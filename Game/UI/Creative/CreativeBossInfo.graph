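# CreativeBossInfo is a layout-only widget on FortCreativeBossInfo, showing a boss's name, shield and health in Creative.
# A size box stacks the gold-material name above an overlay holding shield and health bars, plus a collapsed DBNO (downed) bar.
# The bars are material-instance images from team HUD assets, and shear transforms slant the name and shield/health stack.
# The event graph is empty, so any updates to the inherited text and bar widgets come from outside it.
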WidgetBlueprint /Game/UI/Creative/CreativeBossInfo
graphs [EventGraph {links [], nodes []}]
variables []
widgets {
 info {inheritedWidgetCount=5, parentClass=/Script/FortniteUI.FortCreativeBossInfo, rootWidgetClass=&SizeBox, widgetCount=9}
 widgets [
  SizeBox_23 {widgetClassPath=&SizeBox, widget=SizeBox_23, properties {bOverride_MinDesiredWidth=true, minDesiredWidth=800, widthOverride=200}}
  VerticalBox_53 {widgetClassPath=&VerticalBox, parent=SizeBox_23, widget=VerticalBox_53, slot {class=SizeBoxSlot}}
  Text_PlayerName {
   widgetClassPath=&CommonTextBlock
   bInherited=true
   bIsVariable=true
   parent=VerticalBox_53
   widget=Text_PlayerName
   properties {
    text=TestPlayerName
    font {
     bForceMonospaced=false
     bMaterialIsStencil=false
     fontMaterial=/Game/Athena/HUD/Materials/M_PlayerText_Gold_Inst
     fontObject=&Burbank
     letterSpacing=0
     monospacedWidth=1
     size=14
     skewAmount=0
     typefaceFontName=BurbankRegular-Black
     outlineSettings {
      bApplyOutlineToDropShadows=true
      bMiteredCorners=false
      bSeparateFillAlpha=false
      outlineMaterial=null
      outlineSize=1
      outlineColor=<rgba 0 0 0 0.75>
     }
    }
    margin=<ltrb 0 -2 2 -1>
    renderTransform {angle=0, scale=<1 1>, shear=<-8 0>, translation=<0 0>}
    shadowColorAndOpacity=<rgba 0 0 0 1>
   }
   slot {class=VerticalBoxSlot, properties {horizontalAlignment=HAlign_Center, verticalAlignment=VAlign_Center}}
  }
  Overlay_35 {
   widgetClassPath=&Overlay
   parent=VerticalBox_53
   widget=Overlay_35
   slot {class=VerticalBoxSlot, properties {verticalAlignment=VAlign_Top}}
  }
  VerticalBox_37 {
   widgetClassPath=&VerticalBox
   parent=Overlay_35
   widget=VerticalBox_37
   properties {renderTransform {angle=0, scale=<1 1>, shear=<-8 0>, translation=<0 0>}}
   slot {class=OverlaySlot, properties {horizontalAlignment=HAlign_Fill}}
  }
  Image_ShieldBar {
   widgetClassPath=&Image
   bInherited=true
   bIsVariable=true
   parent=VerticalBox_37
   widget=Image_ShieldBar
   properties {brush {
     bIsDynamicallyLoaded=false
     drawAs=Image
     imageType=NoImage
     mirroring=NoMirror
     resourceName=None
     resourceObject=/Game/Athena/HUD/Team/MI_Player_ShieldBar
     tiling=NoTile
     imageSize=<16 6>
     margin=<ltrb 0 0 0 0>
     outlineSettings {
      bUseBrushTransparency=false
      roundingType=HalfHeightRadius
      width=0
      color {colorUseRule=UseColor_Specified, specifiedColor=<rgba 0 0 0 0>}
      cornerRadii=<0 0 0 0>
     }
     tintColor {colorUseRule=UseColor_Specified, specifiedColor=<rgba 1 1 1 1>}
     uVRegion {bIsValid=false, max=<0 0>, min=<0 0>}
    }}
   slot {class=VerticalBoxSlot, properties {padding=<ltrb 0 0 0 2>}}
  }
  Image_HealthBar {
   widgetClassPath=&Image
   bInherited=true
   bIsVariable=true
   parent=VerticalBox_37
   widget=Image_HealthBar
   properties {brush {
     bIsDynamicallyLoaded=false
     drawAs=Image
     imageType=NoImage
     mirroring=NoMirror
     resourceName=None
     resourceObject=/Game/Athena/HUD/Team/MI_Player_HealthBar
     tiling=NoTile
     imageSize=<16 10>
     margin=<ltrb 0 0 0 0>
     outlineSettings {
      bUseBrushTransparency=false
      roundingType=HalfHeightRadius
      width=0
      color {colorUseRule=UseColor_Specified, specifiedColor=<rgba 0 0 0 0>}
      cornerRadii=<0 0 0 0>
     }
     tintColor {colorUseRule=UseColor_Specified, specifiedColor=<rgba 1 1 1 1>}
     uVRegion {bIsValid=false, max=<0 0>, min=<0 0>}
    }}
   slot {class=VerticalBoxSlot}
  }
  Image_DBNOBar {
   widgetClassPath=&Image
   bInherited=true
   bIsVariable=true
   parent=Overlay_35
   widget=Image_DBNOBar
   properties {
    visibility=Collapsed
    brush {
     bIsDynamicallyLoaded=false
     drawAs=Image
     imageType=NoImage
     mirroring=NoMirror
     resourceName=None
     resourceObject=/Game/Athena/HUD/Team/MI_Player_HealthBar_DBNO
     tiling=NoTile
     imageSize=<24 20>
     margin=<ltrb 0 0 0 0>
     outlineSettings {
      bUseBrushTransparency=false
      roundingType=HalfHeightRadius
      width=0
      color {colorUseRule=UseColor_Specified, specifiedColor=<rgba 0 0 0 0>}
      cornerRadii=<0 0 0 0>
     }
     tintColor {colorUseRule=UseColor_Specified, specifiedColor=<rgba 1 1 1 1>}
     uVRegion {bIsValid=false, max=<0 0>, min=<0 0>}
    }
   }
   slot {class=OverlaySlot, properties {horizontalAlignment=HAlign_Fill}}
  }
  TutorialHighlight {widgetClassPath=&TutorialHighlight}
 ]
}
