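# Creative island details panel widget, built on FortCreativeIslandDetailsPanel, showing an island's image, creator, code, player limit, matchmaking text, name, description and tags.
# Most display widgets are inherited bindings, laid out as a scrollable details box over the image, with a collapsed loading spinner.
# UpdateLinkCodeVisuals shows TagDisplay and fills it from the activity's description tags.
# ClearLinkCodeVisuals repopulates it with a single default tag entry, without changing visibility.
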WidgetBlueprint /Game/UI/Creative/CreativeIslandDetailsPanel
graphs [EventGraph {nodes [
   Event ClearLinkCodeVisuals() {is=title K2Node_Event}
   > PopulateTags!(
    execute
    self:Cr Tag Display Object Reference=Get TagDisplay() -> (TagDisplay:Cr Tag Display Object Reference)
    inTags:Array of Creative Island Description Tag Structures=Make Array("[0]":Creative Island Description Tag Structure) -> (Array:Array of Creative Island Description Tag Structures)
   ) -> (then)
   n1 = Event UpdateLinkCodeVisuals!() -> (OutputDelegate, then, Activity:Fort Game Activity Object Reference) {is=title K2Node_Event}
   n2 = Get TagDisplay()
   > SetVisibility(self=@n2, InVisibility=SelfHitTestInvisible)
   > PopulateTags(
    self=@n2
    inTags=GetActivityDescriptionTags!(self:Fort Game Activity Object Reference=@n1.Activity) -> (ReturnValue:Array of Creative Island Description Tag Structures)
   )
   Event PreConstruct()
   Event Construct()
   Event Tick()
  ]}]
variables []
widgets {
 info {inheritedWidgetCount=9, parentClass=/Script/FortniteUI.FortCreativeIslandDetailsPanel, rootWidgetClass=&Overlay, widgetCount=20}
 widgets [
  Overlay_13 {widgetClassPath=&Overlay, widget=Overlay_13}
  Overlay {
   widgetClassPath=&Overlay
   parent=Overlay_13
   widget=Overlay
   slot {class=OverlaySlot, properties {horizontalAlignment=HAlign_Fill, verticalAlignment=VAlign_Fill}}
  }
  Image_rightside {
   widgetClassPath=&Image
   parent=Overlay_13
   widget=Image_rightside
   properties {
    brush {
     bIsDynamicallyLoaded=false
     drawAs=NoDrawType
     imageType=NoImage
     mirroring=NoMirror
     resourceName=None
     resourceObject=/Game/Athena/UI/InGame/Creative/Art/M_CreativeCreateIslandRightSideBacking
     tiling=NoTile
     imageSize=<999.9958 32>
     margin=<ltrb 0 0 0 0>
     outlineSettings {
      bUseBrushTransparency=false
      roundingType=HalfHeightRadius
      width=0
      color {colorUseRule=UseColor_Specified, specifiedColor=<rgba 0 0 0 0>}
      cornerRadii=<0 0 0 0>
     }
     tintColor {colorUseRule=UseColor_Specified, specifiedColor=<rgba 1 1 1 1>}
     uVRegion {bIsValid=false, max=<0 0>, min=<0 0>}
    }
    colorAndOpacity=<rgba 0 0 0 0.2>
   }
   slot {class=OverlaySlot, properties {horizontalAlignment=HAlign_Right, verticalAlignment=VAlign_Fill}}
  }
  ScaleBox_1 {
   widgetClassPath=&ScaleBox
   parent=Overlay_13
   widget=ScaleBox_1
   properties {stretch=ScaleToFitX}
   slot {class=OverlaySlot, properties {horizontalAlignment=HAlign_Fill}}
  }
  Image_Island {
   widgetClassPath=&Image
   bInherited=true
   bIsVariable=true
   parent=ScaleBox_1
   widget=Image_Island
   properties {brush {
     bIsDynamicallyLoaded=false
     drawAs=Image
     imageType=NoImage
     mirroring=NoMirror
     resourceName=None
     resourceObject=/Game/Playgrounds/Items/Portals/UI/TemplateFeaturedDefaultImage
     tiling=NoTile
     imageSize=<1024 512>
     margin=<ltrb 0 0 0 0>
     outlineSettings {
      bUseBrushTransparency=false
      roundingType=HalfHeightRadius
      width=0
      color {colorUseRule=UseColor_Specified, specifiedColor=<rgba 0 0 0 0>}
      cornerRadii=<0 0 0 0>
     }
     tintColor {colorUseRule=UseColor_Specified, specifiedColor=<rgba 1 1 1 1>}
     uVRegion {bIsValid=false, max=<0 0>, min=<0 0>}
    }}
   slot {class=ScaleBoxSlot}
  }
  Overlay_4 {
   widgetClassPath=&Overlay
   parent=Overlay_13
   widget=Overlay_4
   slot {class=OverlaySlot, properties {horizontalAlignment=HAlign_Fill, verticalAlignment=VAlign_Fill}}
  }
  Image_LoadingSpinner {
   widgetClassPath=&Image
   bInherited=true
   bIsVariable=true
   parent=Overlay_4
   widget=Image_LoadingSpinner
   properties {
    visibility=Collapsed
    brush {
     bIsDynamicallyLoaded=false
     drawAs=Image
     imageType=NoImage
     mirroring=NoMirror
     resourceName=None
     resourceObject=/FortUILibrary/UIKit/Throbber/MI_UIKit_Monochrome_Throbber
     tiling=NoTile
     imageSize=<64 64>
     margin=<ltrb 0 0 0 0>
     outlineSettings {
      bUseBrushTransparency=false
      roundingType=HalfHeightRadius
      width=0
      color {colorUseRule=UseColor_Specified, specifiedColor=<rgba 0 0 0 0>}
      cornerRadii=<0 0 0 0>
     }
     tintColor {colorUseRule=UseColor_Specified, specifiedColor=<rgba 1 1 1 1>}
     uVRegion {bIsValid=false, max=<0 0>, min=<0 0>}
    }
   }
   slot {class=OverlaySlot, properties {horizontalAlignment=HAlign_Center, verticalAlignment=VAlign_Center}}
  }
  Overlay_11 {
   widgetClassPath=&Overlay
   parent=Overlay_4
   widget=Overlay_11
   slot {class=OverlaySlot, properties {horizontalAlignment=HAlign_Fill, verticalAlignment=VAlign_Fill, padding=<ltrb 0 450 0 0>}}
  }
  Image_2 {
   widgetClassPath=&Image
   parent=Overlay_11
   widget=Image_2
   properties {brush {
     bIsDynamicallyLoaded=false
     drawAs=Image
     imageType=NoImage
     mirroring=NoMirror
     resourceName=None
     resourceObject=/Game/Athena/UI/InGame/Creative/Art/M_UI_BoltGradientBorder_Creative_IslandSelectDetailsBacking
     tiling=NoTile
     imageSize=<32 32>
     margin=<ltrb 0 0 0 0>
     outlineSettings {
      bUseBrushTransparency=false
      roundingType=HalfHeightRadius
      width=0
      color {colorUseRule=UseColor_Specified, specifiedColor=<rgba 0 0 0 0>}
      cornerRadii=<0 0 0 0>
     }
     tintColor {colorUseRule=UseColor_Specified, specifiedColor=<rgba 1 1 1 1>}
     uVRegion {bIsValid=false, max=<0 0>, min=<0 0>}
    }}
   slot {class=OverlaySlot, properties {horizontalAlignment=HAlign_Fill, verticalAlignment=VAlign_Fill, padding=<ltrb 0 -32 0 0>}}
  }
  SizeBox_3 {
   widgetClassPath=&SizeBox
   parent=Overlay_11
   widget=SizeBox_3
   properties {bOverride_MaxDesiredHeight=true, heightOverride=300, maxDesiredHeight=300}
   slot {class=OverlaySlot, properties {horizontalAlignment=HAlign_Fill, verticalAlignment=VAlign_Fill, padding=<ltrb 42 0 0 102>}}
  }
  ScrollBox_Details {
   widgetClassPath=&ScrollBox
   bInherited=true
   bIsVariable=true
   parent=SizeBox_3
   widget=ScrollBox_Details
   slot {class=SizeBoxSlot}
  }
  HorizontalBox_5 {widgetClassPath=&HorizontalBox, parent=ScrollBox_Details, widget=HorizontalBox_5, slot {class=ScrollBoxSlot}}
  Text_CreatorLabel {
   widgetClassPath=&CommonTextBlock
   bIsVariable=true
   parent=HorizontalBox_5
   widget=Text_CreatorLabel
   properties {
    text=CREATED BY
    textTransformPolicy=ToUpper
    colorAndOpacity {colorUseRule=UseColor_Specified, specifiedColor=<rgba 0.5271 1 0.8228 1>}
    font {
     bForceMonospaced=false
     bMaterialIsStencil=false
     fontMaterial=null
     fontObject=&Burbank
     letterSpacing=0
     monospacedWidth=1
     size=18
     skewAmount=0
     typefaceFontName=BurbankRegular-Black
     outlineSettings {
      bApplyOutlineToDropShadows=false
      bMiteredCorners=false
      bSeparateFillAlpha=false
      outlineMaterial=null
      outlineSize=0
      outlineColor=<rgba 0 0 0 1>
     }
    }
    renderTransform {angle=0, scale=<1 1>, shear=<-12 0>, translation=<0 0>}
    shadowOffset=<0 0>
   }
   slot {class=HorizontalBoxSlot, properties {padding=<ltrb 0 0 8 0>}}
  }
  Text_Creator {
   widgetClassPath=&CommonTextBlock
   bInherited=true
   bIsVariable=true
   parent=HorizontalBox_5
   widget=Text_Creator
   properties {
    autoWrapText=true
    text="[CREATOR]"
    textTransformPolicy=ToUpper
    colorAndOpacity {colorUseRule=UseColor_Specified, specifiedColor=<rgba 0 1 0.6654 1>}
    font {
     bForceMonospaced=false
     bMaterialIsStencil=false
     fontMaterial=null
     fontObject=&Burbank
     letterSpacing=0
     monospacedWidth=1
     size=18
     skewAmount=0
     typefaceFontName=BurbankRegular-Black
     outlineSettings {
      bApplyOutlineToDropShadows=false
      bMiteredCorners=false
      bSeparateFillAlpha=false
      outlineMaterial=null
      outlineSize=0
      outlineColor=<rgba 0 0 0 1>
     }
    }
    renderTransform {angle=0, scale=<1 1>, shear=<-12 0>, translation=<0 0>}
    shadowOffset=<0 0>
   }
   slot {class=HorizontalBoxSlot, properties {size {sizeRule=Fill, value=1}}}
  }
  Text_PublishedIslandCode {
   widgetClassPath=&CommonTextBlock
   bInherited=true
   bIsVariable=true
   parent=ScrollBox_Details
   widget=Text_PublishedIslandCode
   properties {
    bAutoCollapseWithEmptyText=true
    text=8888-8888-8888
    textTransformPolicy=ToUpper
    colorAndOpacity {colorUseRule=UseColor_Specified, specifiedColor=<rgba 0.502 0.8314 1 1>}
    font {
     bForceMonospaced=false
     bMaterialIsStencil=false
     fontMaterial=null
     fontObject=&Burbank
     letterSpacing=0
     monospacedWidth=1
     size=18
     skewAmount=0
     typefaceFontName=Burbank-Black
     outlineSettings {
      bApplyOutlineToDropShadows=false
      bMiteredCorners=false
      bSeparateFillAlpha=false
      outlineMaterial=null
      outlineSize=0
      outlineColor=<rgba 0 0 0 1>
     }
    }
    shadowOffset=<4.9266 5.2555>
   }
   slot {class=ScrollBoxSlot, properties {verticalAlignment=VAlign_Center}}
  }
  Text_MaximumNumberOfPlayers {
   widgetClassPath=&CommonTextBlock
   bInherited=true
   bIsVariable=true
   parent=ScrollBox_Details
   widget=Text_MaximumNumberOfPlayers
   properties {
    text="Max Players: {0}"
    textTransformPolicy=ToUpper
    colorAndOpacity {colorUseRule=UseColor_Specified, specifiedColor=<rgba 0.502 0.8314 1 1>}
    font {
     bForceMonospaced=false
     bMaterialIsStencil=false
     fontMaterial=null
     fontObject=&Burbank
     letterSpacing=0
     monospacedWidth=1
     size=18
     skewAmount=0
     typefaceFontName=Burbank-Black
     outlineSettings {
      bApplyOutlineToDropShadows=false
      bMiteredCorners=false
      bSeparateFillAlpha=false
      outlineMaterial=null
      outlineSize=0
      outlineColor=<rgba 0 0 0 1>
     }
    }
    shadowOffset=<4.9266 5.2555>
   }
   slot {class=ScrollBoxSlot, properties {verticalAlignment=VAlign_Center}}
  }
  Text_MatchMakingDescription {
   widgetClassPath=&CommonTextBlock
   bInherited=true
   bIsVariable=true
   parent=ScrollBox_Details
   widget=Text_MatchMakingDescription
   properties {
    text=Play with your party and others
    textTransformPolicy=ToUpper
    colorAndOpacity {colorUseRule=UseColor_Specified, specifiedColor=<rgba 0.502 0.8314 1 1>}
    font {
     bForceMonospaced=false
     bMaterialIsStencil=false
     fontMaterial=null
     fontObject=&Burbank
     letterSpacing=0
     monospacedWidth=1
     size=18
     skewAmount=0
     typefaceFontName=Burbank-Black
     outlineSettings {
      bApplyOutlineToDropShadows=false
      bMiteredCorners=false
      bSeparateFillAlpha=false
      outlineMaterial=null
      outlineSize=0
      outlineColor=<rgba 0 0 0 1>
     }
    }
    shadowOffset=<4.9266 5.2555>
   }
   slot {class=ScrollBoxSlot, properties {verticalAlignment=VAlign_Center}}
  }
  Text_IslandName {
   widgetClassPath=&CommonTextBlock
   bInherited=true
   bIsVariable=true
   parent=ScrollBox_Details
   widget=Text_IslandName
   properties {
    autoWrapText=true
    text="[Name]"
    textTransformPolicy=ToUpper
    colorAndOpacity {colorUseRule=UseColor_Specified, specifiedColor=<rgba 0.2195 0.6584 1 1>}
    font {
     bForceMonospaced=false
     bMaterialIsStencil=false
     fontMaterial=/Game/Athena/UI/Challenges/Art/M_UI_Challenge_DynamicColor_MissionHover
     fontObject=&BurbankSmall
     letterSpacing=0
     monospacedWidth=1
     size=24
     skewAmount=0
     typefaceFontName=Black
     outlineSettings {
      bApplyOutlineToDropShadows=false
      bMiteredCorners=false
      bSeparateFillAlpha=false
      outlineMaterial=null
      outlineSize=0
      outlineColor=<rgba 0 0 0 1>
     }
    }
    shadowOffset=<0 0>
   }
   slot {class=ScrollBoxSlot, properties {verticalAlignment=VAlign_Top}}
  }
  Text_IslandDescription {
   widgetClassPath=&CommonTextBlock
   bInherited=true
   bIsVariable=true
   parent=ScrollBox_Details
   widget=Text_IslandDescription
   properties {
    autoWrapText=true
    text="[Description] "
    textTransformPolicy=ToUpper
    colorAndOpacity {colorUseRule=UseColor_Specified, specifiedColor=<rgba 0.0331 0.3968 1 1>}
    font {
     bForceMonospaced=false
     bMaterialIsStencil=false
     fontMaterial=null
     fontObject=&BurbankSmall
     letterSpacing=0
     monospacedWidth=1
     size=18
     skewAmount=0
     typefaceFontName=Black
     outlineSettings {
      bApplyOutlineToDropShadows=false
      bMiteredCorners=false
      bSeparateFillAlpha=false
      outlineMaterial=null
      outlineSize=0
      outlineColor=<rgba 0 0 0 1>
     }
    }
    shadowOffset=<0 0>
   }
   slot {class=ScrollBoxSlot, properties {verticalAlignment=VAlign_Top}}
  }
  TagDisplay {
   widgetClassPath=/Game/Athena/UI/InGame/Creative/Cr_TagDisplay
   bIsVariable=true
   parent=ScrollBox_Details
   widget=TagDisplay
   slot {class=ScrollBoxSlot, properties {padding=<ltrb 0 4 0 0>}}
  }
 ]
}
