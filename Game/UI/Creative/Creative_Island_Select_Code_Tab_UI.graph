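# Widget Blueprint for the Creative island-select tab where a player enters an island code, alongside a details panel and Favorite, Launch and hidden Cancel buttons.
# The code field's change and commit events pass text to the native FortCreativeIslandSelectCodeTab parent's handlers.
# Its validation event shows a LightningViolator banner, using Select nodes to pick message and error or success color by result enum.
# Clear and empty-link events collapse that banner.
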WidgetBlueprint /Game/UI/Creative/Creative_Island_Select_Code_Tab_UI
graphs [EventGraph {nodes [
   n1 = "OnTextChanged (Editable Text) (EditableText_IslandLink)"!() -> (OutputDelegate, then, Text:"Text (by ref)") {is=title K2Node_ComponentBoundEvent}
   > HandleActivityChanged!(execute, self:Fort Creative Island Select Code Tab Object Reference, Text:"Text (by ref)"=@n1.Text) -> (then)
   n2 = "OnTextCommitted (Editable Text) (EditableText_IslandLink)"!() -> (OutputDelegate, then, Text:"Text (by ref)", CommitMethod:ETextCommit Enum=Default) {is=title K2Node_ComponentBoundEvent}
   > HandleActivityCommitted!(
    execute
    self:Fort Creative Island Select Code Tab Object Reference
    InText:"Text (by ref)"=@n2.Text
    CommitInfo:ETextCommit Enum=@n2.CommitMethod
   ) -> (then)
   n3 = Event OnActivityValidated!() -> (OutputDelegate, then, ValidateResult:EFortActivityValidationResult Enum=NotFound) {is=title K2Node_Event}
   n4 = Get LightningViolator() -> (LightningViolator:Lightning Violator Object Reference)
   > SetVisibility(self=@n4, InVisibility=SelfHitTestInvisible)
   > EventSetText!(
    execute
    self:Lightning Violator Object Reference=@n4
    NewText:Text=Select!(
     NotFound:Text="LOCTEXT(\"Try Again - Island Not Found\")"
     InvalidKeyTooShort:Text="LOCTEXT(\"Try Again - Too Short\")"
     InvalidKeyCharacters:Text="LOCTEXT(\"Try Again - Invalid Characters\")"
     IneligibleParty:Text="LOCTEXT(\"Party is too Large\")"
     IslandPrivate:Text="LOCTEXT(\"Island Private\")"
     Success:Text="LOCTEXT(\"ISLAND FOUND!\")"
     Index:EFortActivityValidationResult Enum=@n3.ValidateResult
    ) -> (ReturnValue:Text)
   ) -> (then)
   n5 = Get ViolatorColor_Error() -> (ViolatorColor_Error:Linear Color Structure)
   > EventColorBG!(
    execute
    self:Lightning Violator Object Reference=@n4
    NewColor:Linear Color Structure=Select(
     NotFound:Linear Color Structure=@n5
     InvalidKeyTooShort:Linear Color Structure=@n5
     InvalidKeyCharacters:Linear Color Structure=@n5
     IneligibleParty:Linear Color Structure=@n5
     IslandPrivate:Linear Color Structure=@n5
     Success:Linear Color Structure=Get ViolatorColor_Success() -> (ViolatorColor_Success:Linear Color Structure)
     Index=@n3.ValidateResult
    ) -> (ReturnValue:Linear Color Structure)
   ) -> (then)
   > SetVisibility(self=Get Overlay_Violator() -> (Overlay_Violator:Overlay Object Reference), InVisibility=SelfHitTestInvisible)
   Event HandleActivityClear() {is=title K2Node_Event}
   > SetVisibility(self=Get LightningViolator(), InVisibility=Collapsed)
   Event OnActivityValidatedWithEmptyIslandLink() {is=title K2Node_Event}
   > SetVisibility(self=Get LightningViolator(), InVisibility=Collapsed)
   Event PreConstruct()
   Event Construct()
   Event Tick()
  ]}]
variables [
 InvalidCode {type=WidgetAnimation, default=null}
 ViolatorColor_Error {type=LinearColor, default=<rgba 1 0.0296 0 1>}
 ViolatorColor_Success {type=LinearColor, default=<rgba 0 0.491 1 1>}
]
widgets {
 info {inheritedWidgetCount=5, parentClass=/Script/FortniteUI.FortCreativeIslandSelectCodeTab, rootWidgetClass=&Overlay, widgetCount=29}
 widgets [
  Overlay_21 {widgetClassPath=&Overlay, widget=Overlay_21}
  Overlay_3 {
   widgetClassPath=&Overlay
   parent=Overlay_21
   widget=Overlay_3
   properties {clipping=ClipToBounds}
   slot {class=OverlaySlot, properties {horizontalAlignment=HAlign_Fill, verticalAlignment=VAlign_Fill}}
  }
  HorizontalBox_186 {
   widgetClassPath=&HorizontalBox
   parent=Overlay_3
   widget=HorizontalBox_186
   slot {class=OverlaySlot, properties {horizontalAlignment=HAlign_Fill, verticalAlignment=VAlign_Fill}}
  }
  Overlay_101 {widgetClassPath=&Overlay, parent=HorizontalBox_186, widget=Overlay_101, slot {class=HorizontalBoxSlot}}
  SizeBox {
   widgetClassPath=&SizeBox
   parent=Overlay_101
   widget=SizeBox
   properties {bOverride_WidthOverride=true, widthOverride=860}
   slot {class=OverlaySlot, properties {verticalAlignment=VAlign_Fill}}
  }
  IslandDetailsPanel {
   widgetClassPath=&/CreativeIslandDetailsPanel
   bInherited=true
   bIsVariable=true
   parent=HorizontalBox_186
   widget=IslandDetailsPanel
   slot {class=HorizontalBoxSlot, properties {size {sizeRule=Fill, value=1}}}
  }
  Image_leftside {
   widgetClassPath=&Image
   parent=Overlay_3
   widget=Image_leftside
   properties {
    brush {
     bIsDynamicallyLoaded=false
     drawAs=Image
     imageType=NoImage
     mirroring=NoMirror
     resourceName=None
     resourceObject=/Game/Athena/UI/InGame/Creative/Art/M_CreativeCreateIslandRightSideBacking
     tiling=NoTile
     imageSize=<1020 589.9893>
     margin=<ltrb 0 0 0 0>
     outlineSettings {
      bUseBrushTransparency=false
      roundingType=HalfHeightRadius
      width=0
      color {colorUseRule=UseColor_Specified, specifiedColor=<rgba 0 0 0 0>}
      cornerRadii=<0 0 0 0>
     }
     tintColor {colorUseRule=UseColor_Specified, specifiedColor=<rgba 1 1 1 1>}
     uVRegion {bIsValid=false, max=<0 0>, min=<0 0>}
    }
    colorAndOpacity=<rgba 0 0.0222 0.2874 1>
    renderTransform {angle=0, scale=<-1 -1>, shear=<0 0>, translation=<0 0>}
   }
   slot {class=OverlaySlot, properties {verticalAlignment=VAlign_Fill, padding=<ltrb 0 0 0 -20>}}
  }
  SizeBox_1 {
   widgetClassPath=&SizeBox
   parent=Overlay_3
   widget=SizeBox_1
   properties {bOverride_WidthOverride=true, widthOverride=860}
   slot {class=OverlaySlot, properties {verticalAlignment=VAlign_Fill}}
  }
  Overlay_EnterCode {widgetClassPath=&Overlay, parent=SizeBox_1, widget=Overlay_EnterCode, slot {class=SizeBoxSlot}}
  VerticalBox_0 {
   widgetClassPath=&VerticalBox
   parent=Overlay_EnterCode
   widget=VerticalBox_0
   slot {class=OverlaySlot, properties {horizontalAlignment=HAlign_Fill, verticalAlignment=VAlign_Center, padding=<ltrb 0 -25 0 0>}}
  }
  Text_Islands {
   widgetClassPath=&CommonTextBlock
   parent=VerticalBox_0
   widget=Text_Islands
   properties {
    justification=Center
    text=ENTER AN ISLAND CODE
    textTransformPolicy=ToUpper
    font {
     bForceMonospaced=false
     bMaterialIsStencil=false
     fontMaterial=/Game/Athena/UI/Challenges/Art/M_UI_Challenge_DynamicColor_MissionHover
     fontObject=&Burbank
     letterSpacing=0
     monospacedWidth=1
     size=32
     skewAmount=0
     typefaceFontName=BurbankRegular-Black
     outlineSettings {
      bApplyOutlineToDropShadows=false
      bMiteredCorners=false
      bSeparateFillAlpha=false
      outlineMaterial=null
      outlineSize=0
      outlineColor=<rgba 0 0 0 1>
     }
    }
    renderTransform {angle=0, scale=<1 1>, shear=<-12 0>, translation=<0 0>}
    shadowOffset=<0 0>
   }
   slot {class=VerticalBoxSlot, properties {horizontalAlignment=HAlign_Center, verticalAlignment=VAlign_Center}}
  }
  Spacer_1215 {widgetClassPath=&Spacer, parent=VerticalBox_0, widget=Spacer_1215, properties {size=<1 42>}, slot {class=VerticalBoxSlot}}
  SizeBox_0 {
   widgetClassPath=&SizeBox
   parent=VerticalBox_0
   widget=SizeBox_0
   properties {bOverride_MinDesiredWidth=true, minDesiredWidth=524, widthOverride=600}
   slot {class=VerticalBoxSlot, properties {horizontalAlignment=HAlign_Center, padding=<ltrb 50 0 50 0>}}
  }
  Overlay_Code {widgetClassPath=&Overlay, bIsVariable=true, parent=SizeBox_0, widget=Overlay_Code, slot {class=SizeBoxSlot}}
  Image_181 {
   widgetClassPath=&Image
   parent=Overlay_Code
   widget=Image_181
   properties {colorAndOpacity=<rgba 1 1 1 0.2>, renderTransform {angle=0, scale=<1 1>, shear=<-4 0>, translation=<0 0>}}
   slot {class=OverlaySlot, properties {horizontalAlignment=HAlign_Fill, verticalAlignment=VAlign_Fill}}
  }
  EditableText_IslandLink {
   widgetClassPath=/Script/UMG.EditableText
   bInherited=true
   bIsVariable=true
   parent=Overlay_Code
   widget=EditableText_IslandLink
   properties {
    clearKeyboardFocusOnCommit=false
    hintText=0000-0000-0000
    widgetStyle {
     backgroundImageComposing {
      bIsDynamicallyLoaded=false
      drawAs=Image
      imageType=FullColor
      mirroring=NoMirror
      resourceName=../../../Engine/Content/Slate/Old/HyperlinkDotted.png
      resourceObject=null
      tiling=NoTile
      imageSize=<0 0>
      margin=<ltrb 0 0 0 0.1875>
      outlineSettings {
       bUseBrushTransparency=false
       roundingType=HalfHeightRadius
       width=0
       color {colorUseRule=UseColor_Specified, specifiedColor=<rgba 0 0 0 0>}
       cornerRadii=<0 0 0 0>
      }
      tintColor {colorUseRule=UseColor_Specified, specifiedColor=<rgba 1 1 1 1>}
      uVRegion {bIsValid=false, max=<0 0>, min=<0 0>}
     }
     backgroundImageSelected {
      bIsDynamicallyLoaded=false
      drawAs=Box
      imageType=FullColor
      mirroring=NoMirror
      resourceName=../../../Engine/Content/Slate/Common/EditableTextSelectionBackground.png
      resourceObject=null
      tiling=NoTile
      imageSize=<0 0>
      margin=<ltrb 0.25 0.25 0.25 0.25>
      outlineSettings {
       bUseBrushTransparency=false
       roundingType=HalfHeightRadius
       width=0
       color {colorUseRule=UseColor_Specified, specifiedColor=<rgba 0 0 0 0>}
       cornerRadii=<0 0 0 0>
      }
      tintColor {colorUseRule=UseColor_Specified, specifiedColor=<rgba 0 0.0075 0.0423 1>}
      uVRegion {bIsValid=false, max=<0 0>, min=<0 0>}
     }
     caretImage {
      bIsDynamicallyLoaded=false
      drawAs=Image
      imageType=FullColor
      mirroring=NoMirror
      resourceName=../../../Engine/Content/Slate/Old/White.png
      resourceObject=null
      tiling=NoTile
      imageSize=<16 16>
      margin=<ltrb 0 0 0 0>
      outlineSettings {
       bUseBrushTransparency=false
       roundingType=HalfHeightRadius
       width=0
       color {colorUseRule=UseColor_Specified, specifiedColor=<rgba 0 0 0 0>}
       cornerRadii=<0 0 0 0>
      }
      tintColor {colorUseRule=UseColor_Specified, specifiedColor=<rgba 1 1 1 1>}
      uVRegion {bIsValid=false, max=<0 0>, min=<0 0>}
     }
     colorAndOpacity {colorUseRule=UseColor_Specified, specifiedColor=<rgba 0 0.491 1 1>}
     font {
      bForceMonospaced=false
      bMaterialIsStencil=false
      fontMaterial=null
      fontObject=&Burbank
      letterSpacing=0
      monospacedWidth=1
      size=58
      skewAmount=0
      typefaceFontName=Burbank-Black
      outlineSettings {
       bApplyOutlineToDropShadows=false
       bMiteredCorners=false
       bSeparateFillAlpha=false
       outlineMaterial=null
       outlineSize=0
       outlineColor=<rgba 0 0 0 1>
      }
     }
     hintColorAndOpacity {colorUseRule=UseColor_Foreground, specifiedColor=<rgba 1 0 1 1>}
     selectedBackgroundColor {colorUseRule=UseColor_Foreground, specifiedColor=<rgba 1 0 1 1>}
    }
   }
   slot {class=OverlaySlot, properties {horizontalAlignment=HAlign_Center, verticalAlignment=VAlign_Bottom, padding=<ltrb 12 16 12 0>}}
  }
  Text_WebURL {
   widgetClassPath=&CommonTextBlock
   bIsVariable=true
   parent=Overlay_Code
   widget=Text_WebURL
   properties {
    justification=Center
    text=https://epicgames.com/fn/
    visibility=HitTestInvisible
    colorAndOpacity {colorUseRule=UseColor_Specified, specifiedColor=<rgba 0.2805 0.3364 0.4852 1>}
    font {
     bForceMonospaced=false
     bMaterialIsStencil=false
     fontMaterial=null
     fontObject=&BurbankSmall
     letterSpacing=0
     monospacedWidth=1
     size=16
     skewAmount=0
     typefaceFontName=Black
     outlineSettings {
      bApplyOutlineToDropShadows=false
      bMiteredCorners=false
      bSeparateFillAlpha=false
      outlineMaterial=null
      outlineSize=0
      outlineColor=<rgba 0 0 0 1>
     }
    }
    shadowOffset=<0 0>
   }
   slot {class=OverlaySlot, properties {horizontalAlignment=HAlign_Center}}
  }
  Overlay_Violator {
   widgetClassPath=&Overlay
   bIsVariable=true
   parent=VerticalBox_0
   widget=Overlay_Violator
   properties {visibility=Collapsed}
   slot {class=VerticalBoxSlot, properties {verticalAlignment=VAlign_Bottom, padding=<ltrb 0 12 0 0>}}
  }
  LightningViolator {
   widgetClassPath=/Game/UI/Foundation/Bangs/LightningViolator
   bIsVariable=true
   parent=Overlay_Violator
   widget=LightningViolator
   properties {showOnTop=true, tailHorizontalAlignment=HAlign_Center}
   slot {class=OverlaySlot, properties {horizontalAlignment=HAlign_Center, verticalAlignment=VAlign_Center}}
  }
  Image_1 {
   widgetClassPath=&Image
   bIsVariable=true
   parent=Overlay_3
   widget=Image_1
   properties {
    brush {
     bIsDynamicallyLoaded=false
     drawAs=Image
     imageType=NoImage
     mirroring=NoMirror
     resourceName=None
     resourceObject=/Game/Athena/UI/InGame/Creative/Art/M_UI_BoltGradientBorder_Creative_IslandSelectFooter
     tiling=NoTile
     imageSize=<32 64>
     margin=<ltrb 0 0 0 0>
     outlineSettings {
      bUseBrushTransparency=false
      roundingType=HalfHeightRadius
      width=0
      color {colorUseRule=UseColor_Specified, specifiedColor=<rgba 0 0 0 0>}
      cornerRadii=<0 0 0 0>
     }
     tintColor {colorUseRule=UseColor_Specified, specifiedColor=<rgba 1 1 1 1>}
     uVRegion {bIsValid=false, max=<0 0>, min=<0 0>}
    }
    colorAndOpacity=<rgba 0.0052 0.0578 0.3916 1>
   }
   slot {class=OverlaySlot, properties {horizontalAlignment=HAlign_Fill, verticalAlignment=VAlign_Bottom, padding=<ltrb -4 0 -4 -2>}}
  }
  Overlay_2 {
   widgetClassPath=&Overlay
   parent=Overlay_21
   widget=Overlay_2
   slot {class=OverlaySlot, properties {horizontalAlignment=HAlign_Fill, verticalAlignment=VAlign_Bottom}}
  }
  HorizontalBox_0 {
   widgetClassPath=&HorizontalBox
   parent=Overlay_2
   widget=HorizontalBox_0
   slot {class=OverlaySlot, properties {horizontalAlignment=HAlign_Fill, verticalAlignment=VAlign_Bottom, padding=<ltrb 24 0 24 0>}}
  }
  HBox_ListActions {
   widgetClassPath=&HorizontalBox
   bIsVariable=true
   parent=HorizontalBox_0
   widget=HBox_ListActions
   properties {visibility=Visible}
   slot {class=HorizontalBoxSlot}
  }
  Button_Favorite {
   widgetClassPath=&IconTextButton
   bInherited=true
   bIsVariable=true
   parent=HBox_ListActions
   widget=Button_Favorite
   properties {
    bIsFocusable=false
    buttonText=TOGGLE FAVORITE
    contentAlignment=HAlign_Center
    onButtonBaseClicked=(&self:WidgetTree.Button_Favorite.HandleOnButtonClicked)
    onButtonBaseDoubleClicked=(&self:WidgetTree.Button_Favorite.HandleOnButtonDoubleClicked)
    onButtonBaseHovered=(&self:WidgetTree.Button_Favorite.HandleOnButtonHovered)
    onButtonBaseUnhovered=(&self:WidgetTree.Button_Favorite.HandleOnButtonUnhovered)
    onSelectedChangedBase=(&self:WidgetTree.Button_Favorite.HandleOnSelectedChanged)
    style=/FortUILibrary/Foundation/Buttons/ButtonStyles/Athena/Athena_ButtonStyle_AngledBlueMenuButton.Athena_ButtonStyle_AngledBlueMenuButton_C
    iconBrush {
     bIsDynamicallyLoaded=false
     drawAs=NoDrawType
     imageType=NoImage
     mirroring=NoMirror
     resourceName=None
     resourceObject=null
     tiling=NoTile
     imageSize=<150 150>
     margin=<ltrb 0 0 0 0>
     outlineSettings {
      bUseBrushTransparency=false
      roundingType=HalfHeightRadius
      width=0
      color {colorUseRule=UseColor_Specified, specifiedColor=<rgba 0 0 0 0>}
      cornerRadii=<0 0 0 1>
     }
     tintColor {colorUseRule=UseColor_Specified, specifiedColor=<rgba 1 1 1 1>}
     uVRegion {bIsValid=false, max=<0 0>, min=<0 0>}
    }
    triggeringInputAction {dataTable=&InputActionDataTable, rowName=GenericRightStickClick}
   }
   slot {class=HorizontalBoxSlot, properties {horizontalAlignment=HAlign_Left, verticalAlignment=VAlign_Bottom, padding=<ltrb 0 0 10 16>}}
  }
  MatchmakingSpinner {
   widgetClassPath=&Image
   bIsVariable=true
   parent=HorizontalBox_0
   widget=MatchmakingSpinner
   properties {
    renderOpacity=0
    brush {
     bIsDynamicallyLoaded=false
     drawAs=Image
     imageType=NoImage
     mirroring=NoMirror
     resourceName=None
     resourceObject=/FortUILibrary/UIKit/Throbber/MI_UIKit_Monochrome_Throbber
     tiling=NoTile
     imageSize=<40 40>
     margin=<ltrb 0 0 0 0>
     outlineSettings {
      bUseBrushTransparency=false
      roundingType=HalfHeightRadius
      width=0
      color {colorUseRule=UseColor_Specified, specifiedColor=<rgba 0 0 0 0>}
      cornerRadii=<0 0 0 0>
     }
     tintColor {colorUseRule=UseColor_Specified, specifiedColor=<rgba 1 1 1 1>}
     uVRegion {bIsValid=false, max=<0 0>, min=<0 0>}
    }
    colorAndOpacity=<rgba 1 0.9301 0 1>
   }
   slot {class=HorizontalBoxSlot, properties {horizontalAlignment=HAlign_Center, verticalAlignment=VAlign_Bottom, padding=<ltrb 4 0 4 26>}}
  }
  Overlay_7 {
   widgetClassPath=&Overlay
   parent=HorizontalBox_0
   widget=Overlay_7
   slot {
    class=HorizontalBoxSlot
    properties {horizontalAlignment=HAlign_Right, verticalAlignment=VAlign_Bottom, padding=<ltrb 0 0 15 16>, size {sizeRule=Fill, value=1}}
   }
  }
  Button_Select {
   widgetClassPath=&IconTextButton
   bInherited=true
   bIsVariable=true
   parent=Overlay_7
   widget=Button_Select
   properties {
    bIsFocusable=false
    buttonText=LAUNCH
    contentAlignment=HAlign_Center
    minHeight=72
    minWidth=256
    onButtonBaseClicked=(&self:WidgetTree.Button_Select.HandleOnButtonClicked)
    onButtonBaseDoubleClicked=(&self:WidgetTree.Button_Select.HandleOnButtonDoubleClicked)
    onButtonBaseHovered=(&self:WidgetTree.Button_Select.HandleOnButtonHovered)
    onButtonBaseUnhovered=(&self:WidgetTree.Button_Select.HandleOnButtonUnhovered)
    onSelectedChangedBase=(&self:WidgetTree.Button_Select.HandleOnSelectedChanged)
    style=/FortUILibrary/Foundation/Buttons/ButtonStyles/ButtonStyle_Feature_M_Yellow.ButtonStyle_Feature_M_Yellow_C
    visibility=Visible
    triggeringInputAction {dataTable=&InputActionDataTable, rowName=Launch}
   }
   slot {class=OverlaySlot, properties {horizontalAlignment=HAlign_Right}}
  }
  Widget_LaunchError {
   widgetClassPath=&LightningViolator
   bInherited=true
   bIsVariable=true
   parent=Overlay_7
   widget=Widget_LaunchError
   properties {
    defaultViolatorText=Party too large
    showOnTop=true
    tailHorizontalAlignment=HAlign_Center
    visibility=Collapsed
    renderTransform {angle=0, scale=<1 1>, shear=<0 0>, translation=<-10 91>}
   }
   slot {class=OverlaySlot, properties {horizontalAlignment=HAlign_Center, verticalAlignment=VAlign_Center}}
  }
  Button_Cancel {
   widgetClassPath=&IconTextButton
   bIsVariable=true
   parent=HorizontalBox_0
   widget=Button_Cancel
   properties {
    bIsFocusable=false
    buttonText=CANCEL
    contentAlignment=HAlign_Center
    minWidth=172
    onButtonBaseClicked=(&self:WidgetTree.Button_Cancel.HandleOnButtonClicked)
    onButtonBaseDoubleClicked=(&self:WidgetTree.Button_Cancel.HandleOnButtonDoubleClicked)
    onButtonBaseHovered=(&self:WidgetTree.Button_Cancel.HandleOnButtonHovered)
    onButtonBaseUnhovered=(&self:WidgetTree.Button_Cancel.HandleOnButtonUnhovered)
    onSelectedChangedBase=(&self:WidgetTree.Button_Cancel.HandleOnSelectedChanged)
    style=&Athena_ButtonStyle_AngledBlueMenuButton_C
    visibility=Hidden
    triggeringInputAction {dataTable=/FortUILibrary/Input/CreativeInputActionDataTable, rowName=Cancel}
   }
   slot {class=HorizontalBoxSlot, properties {horizontalAlignment=HAlign_Right, verticalAlignment=VAlign_Bottom, padding=<ltrb 0 0 0 16>}}
  }
 ]
}
