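# List entry widget, derived from FortCreativeIslandLinkEntry, showing one Creative activity's uppercased title and favorite heart.
# Setting the list item caches its FortGameActivity link and mnemonic, trims the title at the first CRLF and switches the heart by favorite status.
# Selection and hover swap the backplate for a highlight, faint on hover.
# Creator and tagline are filled but sit in collapsed VBox_DeleteMe; the Favorite function is empty.
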WidgetBlueprint /Game/UI/Creative/ActivitySetPortalEntry
graphs [
 EventGraph {nodes [
   Event On Selected()
   > Set isSelected(isSelected:bool=true) -> (Output_Get:bool)
   n1 = Get Image_SelectedBG() -> (Image_SelectedBG:Image Object Reference)
   > SetRenderOpacity(self=@n1, InOpacity=1)
   > SetVisibility(self=@n1, InVisibility=SelfHitTestInvisible)
   n2 = Get Image_Backplate() -> (Image_Backplate:Image Object Reference)
   > SetVisibility(self=@n2, InVisibility=Hidden)
   Event On Deselected()
   > Set isSelected(isSelected=false)
   > SetRenderOpacity(self=@n1, InOpacity=1)
   > SetVisibility(self=@n1, InVisibility=Collapsed)
   > SetVisibility(self=@n2, InVisibility=SelfHitTestInvisible)
   Event On Hovered()
   > n3 = Branch(Condition=Get isSelected() -> (isSelected:bool))
   n4 = Get Image_SelectedBG()
   SetRenderOpacity(execute=@n3.else, self=@n4, InOpacity=0.3)
   > SetVisibility(self=@n4, InVisibility=SelfHitTestInvisible)
   n5 = Get Image_Backplate()
   > SetVisibility(self=@n5, InVisibility=Hidden)
   Event On Unhovered()
   > n6 = Branch(Condition=Get isSelected())
   SetRenderOpacity(execute=@n6.else, self=@n4, InOpacity=1)
   > SetVisibility(self=@n4, InVisibility=Collapsed)
   > SetVisibility(self=@n5, InVisibility=SelfHitTestInvisible)
   n7 = Event OnListItemObjectSet()
   > n8 = Cast To FortGameActivity!(execute, Object=@n7.ListItemObject) -> (then, CastFailed, AsFort Game Activity:Fort Game Activity Object Reference) {is=title K2Node_DynamicCast}
   > n9 = Set Link(Link:Fort Game Activity Object Reference=@n8) -> (Output_Get:Fort Game Activity Object Reference)
   > UpdateFavoriteStatus!(execute, self, inActivity:Fort Game Activity Object Reference=@n9) -> (then)
   n10 = Get Link() -> (Link:Fort Game Activity Object Reference)
   > Set Mnemonic(Mnemonic:String=Break Online Link Id!(OnlineLinkId:"Online Link Id Structure (by ref)"=Get LinkId!(self:Fort Game Activity Object Reference=@n10) -> (LinkId:Online Link Id Structure)) -> (Mnemonic:String, Version:int) {is=title K2Node_BreakStruct}) -> (Output_Get:String)
   n11 = "To Text (String)"(InString=Get Creator!(self:Fort Game Activity Object Reference=@n10) -> (Creator:String))
   > "SetText (Text)"(
    self=Get Text_CreatorName() -> (Text_CreatorName:Common Text Object Reference)
    InText=Format Text!(Format="LOCTEXT(\"[ {CreatorUsername} ]\")", CreatorUsername:Text=TextToUpper!(InText:"Text (by ref)"=@n11) -> (ReturnValue:Text)) -> (Result)
   )
   > SetVisibility(self=Get Text_CreatorName(), InVisibility=Select(Option 0=SelfHitTestInvisible, Option 1=Collapsed, Index=TextIsEmpty(InText=@n11)))
   n12 = Get Title!(self:Fort Game Activity Object Reference=Get Link()) -> (Title:String)
   n13 = FindSubstring(SearchIn=@n12, Substring="\r\n")
   > n14 = Branch(Condition="integer >= integer"!(A:int=@n13, B:int=0) -> (ReturnValue:bool) {is=title K2Node_CallFunction})
   > n15 = Set Title(Title:String=GetSubstring!(SourceString:String=@n12, StartIndex:int=0, Length:int=@n13) -> (ReturnValue:String)) -> (Output_Get:String)
   n16 = Get Title!() -> (Title)
   "SetText (Text)"(
    execute=[@n15.then, @n18.then]
    self=Get Text_IslandName() -> (Text_IslandName:Common Text Object Reference)
    InText=TextToUpper(InText="To Text (String)"(InString=@n16))
   )
   n17 = Get Link()
   > "SetText (Text)"(
    self=Get Text_IslandDescription() -> (Text_IslandDescription:Common Text Object Reference)
    InText="To Text (String)"(InString=Get Tagline!(self:Fort Game Activity Object Reference=@n17) -> (Tagline:String))
   )
   n18 = Set Title(execute=@n14.else, Title=Get Title!(self:Fort Game Activity Object Reference=Get Link()) -> (Title))
  ]}
 Favorite {nodes [Favorite!() -> (then, Favorite:bool=~, Link:Fort Creative Island Link Object Reference) {is=title K2Node_FunctionEntry}]}
 UpdateFavoriteStatus {nodes [
   n1 = UpdateFavoriteStatus!() -> (then, inActivity:Fort Game Activity Object Reference) {is=title K2Node_FunctionEntry}
   n2 = GetOwningLocalPlayer!(self:Widget Object Reference) -> (ReturnValue:Local Player Object Reference)
   > SetActiveWidgetIndex(
    self=Get Switcher_Favorite() -> (Switcher_Favorite:Common Widget Switcher Legacy Object Reference)
    Index=Select(
     Option 0:int=0
     Option 1:int=1
     Index=GetIsFavorite!(self:Fort Game Activity Object Reference=@n1, PlayerContext:Local Player Object Reference=@n2) -> (ReturnValue:bool)
    ) -> (ReturnValue:int)
   )
  ]}
]
variables [
 Favorite {}
 isSelected {type=boolean}
 Link {type=FortGameActivity, default=null}
 Mnemonic {type=string, default=""}
 Title {type=string, default=""}
]
widgets {
 info {inheritedWidgetCount=1, parentClass=/Script/FortniteUI.FortCreativeIslandLinkEntry, rootWidgetClass=&Overlay, widgetCount=21}
 widgets [
  Overlay_0 {widgetClassPath=&Overlay, widget=Overlay_0}
  Image_Backplate {
   widgetClassPath=&Image
   bIsVariable=true
   parent=Overlay_0
   widget=Image_Backplate
   properties {visibility=SelfHitTestInvisible, colorAndOpacity=<rgba 1 1 1 0.1>}
   slot {class=OverlaySlot, properties {horizontalAlignment=HAlign_Fill, verticalAlignment=VAlign_Fill, padding=<ltrb 14 0 14 0>}}
  }
  Image_SelectedBG {
   widgetClassPath=&Image
   bIsVariable=true
   parent=Overlay_0
   widget=Image_SelectedBG
   properties {
    visibility=Collapsed
    brush {
     bIsDynamicallyLoaded=false
     drawAs=Image
     imageType=NoImage
     mirroring=NoMirror
     resourceName=None
     resourceObject=/Game/Athena/UI/InGame/Creative/Art/M_SimpleSkewBox_CreativeIslandListSelection
     tiling=NoTile
     imageSize=<32 32>
     margin=<ltrb 0 0 0 0>
     outlineSettings {
      bUseBrushTransparency=false
      roundingType=HalfHeightRadius
      width=0
      color {colorUseRule=UseColor_Specified, specifiedColor=<rgba 0 0 0 0>}
      cornerRadii=<0 0 0 0>
     }
     tintColor {colorUseRule=UseColor_Specified, specifiedColor=<rgba 1 1 1 1>}
     uVRegion {bIsValid=false, max=<0 0>, min=<0 0>}
    }
    renderTransform {angle=0, scale=<1.3 1.3>, shear=<0 0>, translation=<0 0>}
   }
   slot {class=OverlaySlot, properties {horizontalAlignment=HAlign_Fill, verticalAlignment=VAlign_Fill}}
  }
  Border_Text {
   widgetClassPath=&Border
   bIsVariable=true
   parent=Overlay_0
   widget=Border_Text
   properties {
    verticalAlignment=VAlign_Center
    background {
     bIsDynamicallyLoaded=false
     drawAs=NoDrawType
     imageType=NoImage
     mirroring=NoMirror
     resourceName=None
     resourceObject=null
     tiling=NoTile
     imageSize=<32 32>
     margin=<ltrb 0 0 0 0>
     outlineSettings {
      bUseBrushTransparency=false
      roundingType=HalfHeightRadius
      width=0
      color {colorUseRule=UseColor_Specified, specifiedColor=<rgba 0 0 0 0>}
      cornerRadii=<0 0 0 0>
     }
     tintColor {colorUseRule=UseColor_Specified, specifiedColor=<rgba 1 1 1 1>}
     uVRegion {bIsValid=false, max=<0 0>, min=<0 0>}
    }
    padding=<ltrb 0 0 4 0>
   }
   slot {class=OverlaySlot, properties {horizontalAlignment=HAlign_Fill, verticalAlignment=VAlign_Center, padding=<ltrb 24 0 24 0>}}
  }
  HorizontalBox_1 {
   widgetClassPath=&HorizontalBox
   parent=Border_Text
   widget=HorizontalBox_1
   properties {clipping=ClipToBounds}
   slot {class=BorderSlot, properties {verticalAlignment=VAlign_Center, padding=<ltrb 0 0 4 0>}}
  }
  SizeBox_0 {
   widgetClassPath=&SizeBox
   parent=HorizontalBox_1
   widget=SizeBox_0
   properties {clipping=ClipToBounds}
   slot {class=HorizontalBoxSlot, properties {verticalAlignment=VAlign_Center, padding=<ltrb 0 0 24 0>, size {sizeRule=Fill, value=1}}}
  }
  Text_IslandName {
   widgetClassPath=&CommonTextBlock
   bIsVariable=true
   parent=SizeBox_0
   widget=Text_IslandName
   properties {
    text="{Island Name - 1111111111111111111111111}"
    colorAndOpacity {colorUseRule=UseColor_Specified, specifiedColor=<rgba 0 0.5972 1 1>}
    font {
     bForceMonospaced=false
     bMaterialIsStencil=false
     fontMaterial=null
     fontObject=&BurbankSmall
     letterSpacing=0
     monospacedWidth=1
     size=24
     skewAmount=0
     typefaceFontName=Bold
     outlineSettings {
      bApplyOutlineToDropShadows=false
      bMiteredCorners=false
      bSeparateFillAlpha=false
      outlineMaterial=null
      outlineSize=0
      outlineColor=<rgba 0 0 0 1>
     }
    }
    shadowOffset=<0 0>
   }
   slot {class=SizeBoxSlot}
  }
  SizeBox_1 {
   widgetClassPath=&SizeBox
   parent=HorizontalBox_1
   widget=SizeBox_1
   properties {bOverride_HeightOverride=true, bOverride_WidthOverride=true, heightOverride=42, widthOverride=42}
   slot {class=HorizontalBoxSlot, properties {horizontalAlignment=HAlign_Center, verticalAlignment=VAlign_Center}}
  }
  Overlay_3 {widgetClassPath=&Overlay, parent=SizeBox_1, widget=Overlay_3, slot {class=SizeBoxSlot}}
  Image_FavoriteHover {
   widgetClassPath=&Image
   bIsVariable=true
   parent=Overlay_3
   widget=Image_FavoriteHover
   properties {
    visibility=Collapsed
    brush {
     bIsDynamicallyLoaded=false
     drawAs=Image
     imageType=NoImage
     mirroring=NoMirror
     resourceName=None
     resourceObject=/Game/UI/Icons/WorldIcon_Heart
     tiling=NoTile
     imageSize=<32 32>
     margin=<ltrb 0 0 0 0>
     outlineSettings {
      bUseBrushTransparency=false
      roundingType=HalfHeightRadius
      width=0
      color {colorUseRule=UseColor_Specified, specifiedColor=<rgba 0 0 0 0>}
      cornerRadii=<0 0 0 0>
     }
     tintColor {colorUseRule=UseColor_Specified, specifiedColor=<rgba 1 1 1 1>}
     uVRegion {bIsValid=false, max=<0 0>, min=<0 0>}
    }
    colorAndOpacity=<rgba 0 0.5972 1 1>
    renderTransform {angle=0, scale=<1.3 1.3>, shear=<0 0>, translation=<0 0>}
   }
   slot {class=OverlaySlot, properties {horizontalAlignment=HAlign_Center, verticalAlignment=VAlign_Center}}
  }
  Switcher_Favorite {
   widgetClassPath=&CommonWidgetSwitcherLegacy
   bIsVariable=true
   parent=Overlay_3
   widget=Switcher_Favorite
   properties {transitionDuration=0, transitionType=Vertical}
   slot {class=OverlaySlot, properties {horizontalAlignment=HAlign_Center, verticalAlignment=VAlign_Center}}
  }
  Overlay_1 {widgetClassPath=&Overlay, parent=Switcher_Favorite, widget=Overlay_1, slot {class=WidgetSwitcherSlot}}
  Image_NotFavorite_Shadow {
   widgetClassPath=&Image
   parent=Overlay_1
   widget=Image_NotFavorite_Shadow
   properties {
    brush {
     bIsDynamicallyLoaded=false
     drawAs=Image
     imageType=NoImage
     mirroring=NoMirror
     resourceName=None
     resourceObject=&WorldIcon_Heart
     tiling=NoTile
     imageSize=<32 32>
     margin=<ltrb 0 0 0 0>
     outlineSettings {
      bUseBrushTransparency=false
      roundingType=HalfHeightRadius
      width=0
      color {colorUseRule=UseColor_Specified, specifiedColor=<rgba 0 0 0 0>}
      cornerRadii=<0 0 0 0>
     }
     tintColor {colorUseRule=UseColor_Specified, specifiedColor=<rgba 1 1 1 1>}
     uVRegion {bIsValid=false, max=<0 0>, min=<0 0>}
    }
    colorAndOpacity=<rgba 0 0 0 0.2>
   }
   slot {class=OverlaySlot}
  }
  Overlay_2 {widgetClassPath=&Overlay, parent=Switcher_Favorite, widget=Overlay_2, slot {class=WidgetSwitcherSlot}}
  Image_Favorite {
   widgetClassPath=&Image
   parent=Overlay_2
   widget=Image_Favorite
   properties {
    brush {
     bIsDynamicallyLoaded=false
     drawAs=Image
     imageType=NoImage
     mirroring=NoMirror
     resourceName=None
     resourceObject=&WorldIcon_Heart
     tiling=NoTile
     imageSize=<32 32>
     margin=<ltrb 0 0 0 0>
     outlineSettings {
      bUseBrushTransparency=false
      roundingType=HalfHeightRadius
      width=0
      color {colorUseRule=UseColor_Specified, specifiedColor=<rgba 0 0 0 0>}
      cornerRadii=<0 0 0 0>
     }
     tintColor {colorUseRule=UseColor_Specified, specifiedColor=<rgba 1 1 1 1>}
     uVRegion {bIsValid=false, max=<0 0>, min=<0 0>}
    }
    colorAndOpacity=<rgba 0.7991 0.0086 0.1845 1>
   }
   slot {class=OverlaySlot}
  }
  Button_Favorite {
   widgetClassPath=/Game/Athena/UI/Matchmaking/SimpleCommonButton
   bIsVariable=true
   parent=Overlay_3
   widget=Button_Favorite
   properties {
    onButtonBaseClicked=(&self:WidgetTree.Button_Favorite.HandleOnButtonClicked)
    onButtonBaseDoubleClicked=(&self:WidgetTree.Button_Favorite.HandleOnButtonDoubleClicked)
    onButtonBaseHovered=(&self:WidgetTree.Button_Favorite.HandleOnButtonHovered)
    onButtonBaseUnhovered=(&self:WidgetTree.Button_Favorite.HandleOnButtonUnhovered)
    onSelectedChangedBase=(&self:WidgetTree.Button_Favorite.HandleOnSelectedChanged)
    style=/FortUILibrary/Foundation/Buttons/ButtonStyles/ButtonStyle-FullyInvisible.ButtonStyle-FullyInvisible_C
   }
   slot {class=OverlaySlot, properties {horizontalAlignment=HAlign_Fill, verticalAlignment=VAlign_Fill}}
  }
  VBox_DeleteMe {
   widgetClassPath=&VerticalBox
   parent=Overlay_0
   widget=VBox_DeleteMe
   properties {visibility=Collapsed}
   slot {class=OverlaySlot, properties {horizontalAlignment=HAlign_Fill, verticalAlignment=VAlign_Fill}}
  }
  Text_CreatorName {
   widgetClassPath=&CommonTextBlock
   bIsVariable=true
   parent=VBox_DeleteMe
   widget=Text_CreatorName
   properties {
    text="[ Creator Name ]"
    colorAndOpacity {colorUseRule=UseColor_Specified, specifiedColor=<rgba 0 0.5972 1 1>}
    font {
     bForceMonospaced=false
     bMaterialIsStencil=false
     fontMaterial=null
     fontObject=&Burbank
     letterSpacing=0
     monospacedWidth=1
     size=32
     skewAmount=0
     typefaceFontName=Burbank-Black
     outlineSettings {
      bApplyOutlineToDropShadows=false
      bMiteredCorners=false
      bSeparateFillAlpha=false
      outlineMaterial=null
      outlineSize=0
      outlineColor=<rgba 0 0 0 1>
     }
    }
    shadowOffset=<0 0>
   }
   slot {class=VerticalBoxSlot, properties {verticalAlignment=VAlign_Center, padding=<ltrb 0 0 15 0>}}
  }
  Text_IslandDescription {
   widgetClassPath=&CommonTextBlock
   bIsVariable=true
   parent=VBox_DeleteMe
   widget=Text_IslandDescription
   properties {
    clipping=ClipToBounds
    scrollStyle=&TextScrollStyle-Base_C
    style=/FortUILibrary/Styles/Typography/Foundation/TextStyle-Base-M-B-70pc.TextStyle-Base-M-B-70pc_C
    text="{ Island Description - Lorem ipsum dolor sit amet, consectetur adipiscing elit. Sed non nisl felis. Donec non finibus leo. Cras blandit vestibulum odio, et consequat nullam. }"
    colorAndOpacity {colorUseRule=UseColor_Specified, specifiedColor=<rgba 1 1 1 0.7>}
    font {
     bForceMonospaced=false
     bMaterialIsStencil=false
     fontMaterial=null
     fontObject=&NotoSans
     letterSpacing=0
     monospacedWidth=1
     size=24
     skewAmount=0
     typefaceFontName=NotoSans-Bold
     outlineSettings {
      bApplyOutlineToDropShadows=false
      bMiteredCorners=false
      bSeparateFillAlpha=false
      outlineMaterial=null
      outlineSize=0
      outlineColor=<rgba 0 0 0 1>
     }
    }
    shadowOffset=<0 0>
   }
   slot {class=VerticalBoxSlot, properties {padding=<ltrb 60 0 0 0>}}
  }
  SimplePipButton {
   widgetClassPath=/FortUILibrary/Foundation/Buttons/SimplePipButton
   bIsVariable=true
   parent=VBox_DeleteMe
   widget=SimplePipButton
   properties {
    onButtonBaseClicked=(&self:WidgetTree.SimplePipButton.HandleOnButtonClicked)
    onButtonBaseDoubleClicked=(&self:WidgetTree.SimplePipButton.HandleOnButtonDoubleClicked)
    onButtonBaseHovered=(&self:WidgetTree.SimplePipButton.HandleOnButtonHovered)
    onButtonBaseUnhovered=(&self:WidgetTree.SimplePipButton.HandleOnButtonUnhovered)
    onSelectedChangedBase=(&self:WidgetTree.SimplePipButton.HandleOnSelectedChanged)
    colorAndOpacity=<rgba 1 0 0 1>
   }
   slot {class=VerticalBoxSlot, properties {horizontalAlignment=HAlign_Center, verticalAlignment=VAlign_Center}}
  }
  InputActionWidget {widgetClassPath=&CommonActionWidget, bInherited=true, widget=null}
 ]
}
