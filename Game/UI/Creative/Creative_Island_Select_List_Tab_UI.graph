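# Widget Blueprint for a Creative island select list tab, where players browse islands and launch them, built on FortCreativeIslandSelectListTab.
# A switcher shows a list of activity entries, a no-islands hint or a no-favorites message, beside an inherited island details panel.
# A bottom bar holds Favorite, Remove, Launch and Cancel buttons with input actions, plus throbbers and a "Party too large" violator.
# The event graph contains only unconnected stock events.
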WidgetBlueprint /Game/UI/Creative/Creative_Island_Select_List_Tab_UI
graphs [EventGraph {nodes [
   Event PreConstruct()
   Event Construct()
   Event Tick()
  ]}]
variables []
widgets {
 info {inheritedWidgetCount=12, parentClass=/Script/FortniteUI.FortCreativeIslandSelectListTab, rootWidgetClass=&Overlay, widgetCount=31}
 widgets [
  Overlay_30 {widgetClassPath=&Overlay, widget=Overlay_30}
  Overlay_0 {
   widgetClassPath=&Overlay
   parent=Overlay_30
   widget=Overlay_0
   properties {clipping=ClipToBounds}
   slot {class=OverlaySlot, properties {horizontalAlignment=HAlign_Fill, verticalAlignment=VAlign_Fill}}
  }
  Overlay_101 {
   widgetClassPath=&Overlay
   parent=Overlay_0
   widget=Overlay_101
   slot {class=OverlaySlot, properties {horizontalAlignment=HAlign_Fill, verticalAlignment=VAlign_Fill, padding=<ltrb 0 0 700 120>}}
  }
  SizeBox_7 {
   widgetClassPath=&SizeBox
   parent=Overlay_101
   widget=SizeBox_7
   properties {bOverride_WidthOverride=true, widthOverride=860}
   slot {class=OverlaySlot, properties {verticalAlignment=VAlign_Fill, padding=<ltrb 24 0 0 0>}}
  }
  HorizontalBox_186 {
   widgetClassPath=&HorizontalBox
   parent=Overlay_0
   widget=HorizontalBox_186
   slot {class=OverlaySlot, properties {horizontalAlignment=HAlign_Fill, verticalAlignment=VAlign_Fill}}
  }
  Overlay {widgetClassPath=&Overlay, parent=HorizontalBox_186, widget=Overlay, slot {class=HorizontalBoxSlot}}
  SizeBox {
   widgetClassPath=&SizeBox
   parent=Overlay
   widget=SizeBox
   properties {bOverride_WidthOverride=true, widthOverride=860}
   slot {class=OverlaySlot, properties {verticalAlignment=VAlign_Fill}}
  }
  IslandDetailsPanel {
   widgetClassPath=&/CreativeIslandDetailsPanel
   bInherited=true
   bIsVariable=true
   parent=HorizontalBox_186
   widget=IslandDetailsPanel
   slot {class=HorizontalBoxSlot, properties {size {sizeRule=Fill, value=1}}}
  }
  Image_leftside {
   widgetClassPath=&Image
   parent=Overlay_0
   widget=Image_leftside
   properties {
    brush {
     bIsDynamicallyLoaded=false
     drawAs=Image
     imageType=NoImage
     mirroring=NoMirror
     resourceName=None
     resourceObject=/Game/Athena/UI/InGame/Creative/Art/M_CreativeCreateIslandRightSideBacking
     tiling=NoTile
     imageSize=<1020 589.9893>
     margin=<ltrb 0 0 0 0>
     outlineSettings {
      bUseBrushTransparency=false
      roundingType=HalfHeightRadius
      width=0
      color {colorUseRule=UseColor_Specified, specifiedColor=<rgba 0 0 0 0>}
      cornerRadii=<0 0 0 0>
     }
     tintColor {colorUseRule=UseColor_Specified, specifiedColor=<rgba 1 1 1 1>}
     uVRegion {bIsValid=false, max=<0 0>, min=<0 0>}
    }
    colorAndOpacity=<rgba 0 0.0222 0.2874 1>
    renderTransform {angle=0, scale=<-1 -1>, shear=<0 0>, translation=<0 0>}
   }
   slot {class=OverlaySlot, properties {verticalAlignment=VAlign_Fill, padding=<ltrb 0 0 0 -20>}}
  }
  SizeBox_1 {
   widgetClassPath=&SizeBox
   parent=Overlay_0
   widget=SizeBox_1
   properties {bOverride_WidthOverride=true, widthOverride=860}
   slot {class=OverlaySlot, properties {verticalAlignment=VAlign_Fill}}
  }
  Overlay_List {widgetClassPath=&Overlay, parent=SizeBox_1, widget=Overlay_List, slot {class=SizeBoxSlot}}
  Switcher_IslandsList {
   widgetClassPath=&CommonActivatableWidgetSwitcher
   bInherited=true
   bIsVariable=true
   parent=Overlay_List
   widget=Switcher_IslandsList
   properties {transitionDuration=0}
   slot {class=OverlaySlot, properties {horizontalAlignment=HAlign_Fill, verticalAlignment=VAlign_Fill, padding=<ltrb 0 10 0 0>}}
  }
  Overlay_IslandListContent {
   widgetClassPath=&Overlay
   bInherited=true
   bIsVariable=true
   parent=Switcher_IslandsList
   widget=Overlay_IslandListContent
   slot {class=WidgetSwitcherSlot}
  }
  ListView_Activities {
   widgetClassPath=&CommonListView
   bInherited=true
   bIsVariable=true
   parent=Overlay_IslandListContent
   widget=ListView_Activities
   properties {
    entryWidgetClass=&/ActivitySetPortalEntry.ActivitySetPortalEntry_C
    horizontalEntrySpacing=4
    navigation=&self:WidgetTree.ListView_Activities.WidgetNavigation_0
    verticalEntrySpacing=4
   }
   slot {class=OverlaySlot, properties {horizontalAlignment=HAlign_Fill, verticalAlignment=VAlign_Fill, padding=<ltrb 0 0 0 64>}}
  }
  Overlay_NoIslandsContent {
   widgetClassPath=&Overlay
   bInherited=true
   bIsVariable=true
   parent=Switcher_IslandsList
   widget=Overlay_NoIslandsContent
   slot {class=WidgetSwitcherSlot}
  }
  Text_NoIslands {
   widgetClassPath=&CommonTextBlock
   bInherited=true
   bIsVariable=true
   parent=Overlay_NoIslandsContent
   widget=Text_NoIslands
   properties {
    autoWrapText=true
    style=/FortUILibrary/Styles/Typography/Foundation/TextStyle-DecoBase-M.TextStyle-DecoBase-M_C
    text="Use the ENTER CODE tab to type in a code for an island to visit.\r\nPreviously entered codes will show up here as favorites.\r\n\r\n*Island Sharing is in limited access, but soon you'll be able to generate your own island codes to share with friends (and rivals)!*"
    font {
     bForceMonospaced=false
     bMaterialIsStencil=false
     fontMaterial=null
     fontObject=&BurbankSmall
     letterSpacing=0
     monospacedWidth=1
     size=24
     skewAmount=0
     typefaceFontName=Bold
     outlineSettings {
      bApplyOutlineToDropShadows=false
      bMiteredCorners=false
      bSeparateFillAlpha=false
      outlineMaterial=null
      outlineSize=0
      outlineColor=<rgba 0 0 0 1>
     }
    }
    shadowOffset=<0 0>
   }
   slot {class=OverlaySlot, properties {padding=<ltrb 0 50 0 0>}}
  }
  Overlay_NoFavorites {
   widgetClassPath=&Overlay
   bInherited=true
   bIsVariable=true
   parent=Switcher_IslandsList
   widget=Overlay_NoFavorites
   slot {class=WidgetSwitcherSlot}
  }
  Text_NoIslands_1 {
   widgetClassPath=&CommonTextBlock
   bIsVariable=true
   parent=Overlay_NoFavorites
   widget=Text_NoIslands_1
   properties {
    autoWrapText=true
    renderOpacity=0.7
    style=&TextStyle-DecoBase-M_C
    text=No favorites marked
    font {
     bForceMonospaced=false
     bMaterialIsStencil=false
     fontMaterial=null
     fontObject=&BurbankSmall
     letterSpacing=0
     monospacedWidth=1
     size=24
     skewAmount=0
     typefaceFontName=Bold
     outlineSettings {
      bApplyOutlineToDropShadows=false
      bMiteredCorners=false
      bSeparateFillAlpha=false
      outlineMaterial=null
      outlineSize=0
      outlineColor=<rgba 0 0 0 1>
     }
    }
    renderTransform {angle=0, scale=<1 1>, shear=<-12 0>, translation=<0 0>}
    shadowOffset=<0 0>
   }
   slot {class=OverlaySlot, properties {horizontalAlignment=HAlign_Fill, padding=<ltrb 32 32 32 32>}}
  }
  Image_1 {
   widgetClassPath=&Image
   bIsVariable=true
   parent=Overlay_0
   widget=Image_1
   properties {
    brush {
     bIsDynamicallyLoaded=false
     drawAs=Image
     imageType=NoImage
     mirroring=NoMirror
     resourceName=None
     resourceObject=/Game/Athena/UI/InGame/Creative/Art/M_UI_BoltGradientBorder_Creative_IslandSelectFooter
     tiling=NoTile
     imageSize=<32 64>
     margin=<ltrb 0 0 0 0>
     outlineSettings {
      bUseBrushTransparency=false
      roundingType=HalfHeightRadius
      width=0
      color {colorUseRule=UseColor_Specified, specifiedColor=<rgba 0 0 0 0>}
      cornerRadii=<0 0 0 0>
     }
     tintColor {colorUseRule=UseColor_Specified, specifiedColor=<rgba 1 1 1 1>}
     uVRegion {bIsValid=false, max=<0 0>, min=<0 0>}
    }
    colorAndOpacity=<rgba 0.0052 0.0578 0.3916 1>
   }
   slot {class=OverlaySlot, properties {horizontalAlignment=HAlign_Fill, verticalAlignment=VAlign_Bottom, padding=<ltrb -4 0 -4 -2>}}
  }
  Overlay_2 {
   widgetClassPath=&Overlay
   parent=Overlay_30
   widget=Overlay_2
   slot {class=OverlaySlot, properties {horizontalAlignment=HAlign_Fill, verticalAlignment=VAlign_Bottom}}
  }
  HorizontalBox_0 {
   widgetClassPath=&HorizontalBox
   parent=Overlay_2
   widget=HorizontalBox_0
   slot {class=OverlaySlot, properties {horizontalAlignment=HAlign_Fill, verticalAlignment=VAlign_Bottom, padding=<ltrb 24 0 24 0>}}
  }
  HBox_ListActions {
   widgetClassPath=&HorizontalBox
   bIsVariable=true
   parent=HorizontalBox_0
   widget=HBox_ListActions
   properties {visibility=Visible}
   slot {class=HorizontalBoxSlot}
  }
  Button_Favorite {
   widgetClassPath=&IconTextButton
   bInherited=true
   bIsVariable=true
   parent=HBox_ListActions
   widget=Button_Favorite
   properties {
    bIsFocusable=false
    buttonText=TOGGLE FAVORITE
    contentAlignment=HAlign_Center
    onButtonBaseClicked=(&self:WidgetTree.Button_Favorite.HandleOnButtonClicked)
    onButtonBaseDoubleClicked=(&self:WidgetTree.Button_Favorite.HandleOnButtonDoubleClicked)
    onButtonBaseHovered=(&self:WidgetTree.Button_Favorite.HandleOnButtonHovered)
    onButtonBaseUnhovered=(&self:WidgetTree.Button_Favorite.HandleOnButtonUnhovered)
    onSelectedChangedBase=(&self:WidgetTree.Button_Favorite.HandleOnSelectedChanged)
    style=/FortUILibrary/Foundation/Buttons/ButtonStyles/Athena/Athena_ButtonStyle_AngledBlueMenuButton.Athena_ButtonStyle_AngledBlueMenuButton_C
    iconBrush {
     bIsDynamicallyLoaded=false
     drawAs=NoDrawType
     imageType=NoImage
     mirroring=NoMirror
     resourceName=None
     resourceObject=null
     tiling=NoTile
     imageSize=<150 150>
     margin=<ltrb 0 0 0 0>
     outlineSettings {
      bUseBrushTransparency=false
      roundingType=HalfHeightRadius
      width=0
      color {colorUseRule=UseColor_Specified, specifiedColor=<rgba 0 0 0 0>}
      cornerRadii=<0 0 0 1>
     }
     tintColor {colorUseRule=UseColor_Specified, specifiedColor=<rgba 1 1 1 1>}
     uVRegion {bIsValid=false, max=<0 0>, min=<0 0>}
    }
    triggeringInputAction {dataTable=&InputActionDataTable, rowName=GenericRightStickClick}
   }
   slot {class=HorizontalBoxSlot, properties {horizontalAlignment=HAlign_Left, verticalAlignment=VAlign_Bottom, padding=<ltrb 0 0 10 16>}}
  }
  Overlay_Remove {
   widgetClassPath=&Overlay
   bIsVariable=true
   parent=HBox_ListActions
   widget=Overlay_Remove
   slot {class=HorizontalBoxSlot, properties {padding=<ltrb 0 0 10 16>}}
  }
  Button_Remove {
   widgetClassPath=&IconTextButton
   bInherited=true
   bIsVariable=true
   parent=Overlay_Remove
   widget=Button_Remove
   properties {
    bIsFocusable=false
    buttonText=REMOVE
    contentAlignment=HAlign_Center
    minWidth=172
    onButtonBaseClicked=(&self:WidgetTree.Button_Remove.HandleOnButtonClicked)
    onButtonBaseDoubleClicked=(&self:WidgetTree.Button_Remove.HandleOnButtonDoubleClicked)
    onButtonBaseHovered=(&self:WidgetTree.Button_Remove.HandleOnButtonHovered)
    onButtonBaseUnhovered=(&self:WidgetTree.Button_Remove.HandleOnButtonUnhovered)
    onSelectedChangedBase=(&self:WidgetTree.Button_Remove.HandleOnSelectedChanged)
    style=&Athena_ButtonStyle_AngledBlueMenuButton_C
    visibility=Collapsed
    iconBrush {
     bIsDynamicallyLoaded=false
     drawAs=NoDrawType
     imageType=NoImage
     mirroring=NoMirror
     resourceName=None
     resourceObject=null
     tiling=NoTile
     imageSize=<150 150>
     margin=<ltrb 0 0 0 0>
     outlineSettings {
      bUseBrushTransparency=false
      roundingType=HalfHeightRadius
      width=0
      color {colorUseRule=UseColor_Specified, specifiedColor=<rgba 0 0 0 0>}
      cornerRadii=<0 0 0 1>
     }
     tintColor {colorUseRule=UseColor_Specified, specifiedColor=<rgba 1 1 1 1>}
     uVRegion {bIsValid=false, max=<0 0>, min=<0 0>}
    }
    triggeringInputAction {dataTable=&InputActionDataTable, rowName=GenericFaceButtonTop}
   }
   slot {class=OverlaySlot}
  }
  Image_LoadingIndicator {
   widgetClassPath=&Image
   bInherited=true
   bIsVariable=true
   parent=HorizontalBox_0
   widget=Image_LoadingIndicator
   properties {
    visibility=Collapsed
    brush {
     bIsDynamicallyLoaded=false
     drawAs=Image
     imageType=NoImage
     mirroring=NoMirror
     resourceName=None
     resourceObject=/FortUILibrary/UIKit/Throbber/MI_UIKit_Monochrome_Throbber
     tiling=NoTile
     imageSize=<40 40>
     margin=<ltrb 0 0 0 0>
     outlineSettings {
      bUseBrushTransparency=false
      roundingType=HalfHeightRadius
      width=0
      color {colorUseRule=UseColor_Specified, specifiedColor=<rgba 0 0 0 0>}
      cornerRadii=<0 0 0 0>
     }
     tintColor {colorUseRule=UseColor_Specified, specifiedColor=<rgba 1 1 1 1>}
     uVRegion {bIsValid=false, max=<0 0>, min=<0 0>}
    }
   }
   slot {class=HorizontalBoxSlot, properties {horizontalAlignment=HAlign_Center, verticalAlignment=VAlign_Center, padding=<ltrb 8 0 8 12>}}
  }
  MatchmakingSpinner {
   widgetClassPath=&Image
   bIsVariable=true
   parent=HorizontalBox_0
   widget=MatchmakingSpinner
   properties {
    renderOpacity=0
    brush {
     bIsDynamicallyLoaded=false
     drawAs=Image
     imageType=NoImage
     mirroring=NoMirror
     resourceName=None
     resourceObject=&MI_UIKit_Monochrome_Throbber
     tiling=NoTile
     imageSize=<40 40>
     margin=<ltrb 0 0 0 0>
     outlineSettings {
      bUseBrushTransparency=false
      roundingType=HalfHeightRadius
      width=0
      color {colorUseRule=UseColor_Specified, specifiedColor=<rgba 0 0 0 0>}
      cornerRadii=<0 0 0 0>
     }
     tintColor {colorUseRule=UseColor_Specified, specifiedColor=<rgba 1 1 1 1>}
     uVRegion {bIsValid=false, max=<0 0>, min=<0 0>}
    }
    colorAndOpacity=<rgba 1 0.9301 0 1>
   }
   slot {class=HorizontalBoxSlot, properties {horizontalAlignment=HAlign_Center, verticalAlignment=VAlign_Bottom, padding=<ltrb 4 0 4 26>}}
  }
  Overlay_7 {
   widgetClassPath=&Overlay
   parent=HorizontalBox_0
   widget=Overlay_7
   slot {
    class=HorizontalBoxSlot
    properties {horizontalAlignment=HAlign_Right, verticalAlignment=VAlign_Bottom, padding=<ltrb 0 0 15 16>, size {sizeRule=Fill, value=1}}
   }
  }
  Button_Select {
   widgetClassPath=&IconTextButton
   bInherited=true
   bIsVariable=true
   parent=Overlay_7
   widget=Button_Select
   properties {
    bIsFocusable=false
    buttonText=LAUNCH
    contentAlignment=HAlign_Center
    minHeight=72
    minWidth=256
    onButtonBaseClicked=(&self:WidgetTree.Button_Select.HandleOnButtonClicked)
    onButtonBaseDoubleClicked=(&self:WidgetTree.Button_Select.HandleOnButtonDoubleClicked)
    onButtonBaseHovered=(&self:WidgetTree.Button_Select.HandleOnButtonHovered)
    onButtonBaseUnhovered=(&self:WidgetTree.Button_Select.HandleOnButtonUnhovered)
    onSelectedChangedBase=(&self:WidgetTree.Button_Select.HandleOnSelectedChanged)
    style=/FortUILibrary/Foundation/Buttons/ButtonStyles/ButtonStyle_Feature_M_Yellow.ButtonStyle_Feature_M_Yellow_C
    visibility=Visible
    triggeringInputAction {dataTable=&InputActionDataTable, rowName=Launch}
   }
   slot {class=OverlaySlot, properties {horizontalAlignment=HAlign_Right}}
  }
  Widget_LaunchError {
   widgetClassPath=/Game/UI/Foundation/Bangs/LightningViolator
   bInherited=true
   bIsVariable=true
   parent=Overlay_7
   widget=Widget_LaunchError
   properties {
    defaultViolatorText=Party too large
    showOnTop=true
    tailHorizontalAlignment=HAlign_Center
    visibility=Collapsed
    renderTransform {angle=0, scale=<1 1>, shear=<0 0>, translation=<-10 91>}
   }
   slot {class=OverlaySlot, properties {horizontalAlignment=HAlign_Right}}
  }
  Button_Cancel {
   widgetClassPath=&IconTextButton
   bIsVariable=true
   parent=HorizontalBox_0
   widget=Button_Cancel
   properties {
    bIsFocusable=false
    buttonText=CANCEL
    contentAlignment=HAlign_Center
    minWidth=172
    onButtonBaseClicked=(&self:WidgetTree.Button_Cancel.HandleOnButtonClicked)
    onButtonBaseDoubleClicked=(&self:WidgetTree.Button_Cancel.HandleOnButtonDoubleClicked)
    onButtonBaseHovered=(&self:WidgetTree.Button_Cancel.HandleOnButtonHovered)
    onButtonBaseUnhovered=(&self:WidgetTree.Button_Cancel.HandleOnButtonUnhovered)
    onSelectedChangedBase=(&self:WidgetTree.Button_Cancel.HandleOnSelectedChanged)
    style=&Athena_ButtonStyle_AngledBlueMenuButton_C
    visibility=Hidden
    triggeringInputAction {dataTable=/FortUILibrary/Input/CreativeInputActionDataTable, rowName=Cancel}
   }
   slot {class=HorizontalBoxSlot, properties {horizontalAlignment=HAlign_Right, verticalAlignment=VAlign_Bottom, padding=<ltrb 0 0 0 16>}}
  }
 ]
}
