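# Creative-mode UI widget, derived from FortResourceWidget, that shows a resource icon with its cost and signals whether the player can afford it.
# OnResourceCostSet treats count at or above cost, or infinite resources, as affordable and picks the cost text color accordingly.
# With bUseBackgroundForAffordabilityFeedback set, it also tints the Background border; Construct collapses that border otherwise.
# The inherited resource count text is not placed in the layout.
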WidgetBlueprint /Game/UI/Creative/ResourceCountWidget
graphs [EventGraph {nodes [
   n1 = Event OnResourceCostSet!() -> (OutputDelegate, then, ResourceCount:int=0, ResourceCost:int=0, bHasInfiniteResources:bool) {is=title K2Node_Event}
   n2 = OR Boolean(A="integer >= integer"(A=@n1.ResourceCount, B=@n1.ResourceCost), B=@n1.bHasInfiniteResources)
   > SetColorAndOpacity(
    self=Get NumericText_ResourceCost() -> (NumericText_ResourceCost:Common Numeric Text Block Object Reference)
    InColorAndOpacity=Select(
     Option 0:Slate Color Structure=Get CanNotAffordTextColor() -> (CanNotAffordTextColor:Slate Color Structure)
     Option 1:Slate Color Structure=Get CanAffordTextColor() -> (CanAffordTextColor:Slate Color Structure)
     Index=@n2
    ) -> (ReturnValue:Slate Color Structure)
   )
   > Branch(Condition=Get bUseBackgroundForAffordabilityFeedback() -> (bUseBackgroundForAffordabilityFeedback:bool))
   n3 = Get CanNotAffordBackgroundColor() -> (CanNotAffordBackgroundColor:Slate Color Structure)
   n4 = Get CanAffordBackgroundColor() -> (CanAffordBackgroundColor:Slate Color Structure)
   > SetBrush(
    self:Border Object Reference=Get Background() -> (Background:Border Object Reference)
    InBrush=Make Slate Brush!(
     TintColor:Slate Color Structure=Select(Option 0=@n3, Option 1=@n4, Index=@n2)
     DrawAs:ESlateBrushDrawType Enum=Box
     OutlineSettings:Slate Brush Outline Settings Structure=(CornerRadii=(X=0,Y=0,Z=0,W=0),Color=(SpecifiedColor=(R=0,G=0,B=0,A=0),ColorUseRule=UseColor_Specified),Width=0,RoundingType=HalfHeightRadius,bUseBrushTransparency=False)
    ) -> (SlateBrush:Slate Brush Structure) {is=title K2Node_MakeStruct}
   )
   Event Construct()
   > SetVisibility(
    self=Get Background()
    InVisibility=Select(
     Option 0:ESlateVisibility Enum=Collapsed
     Option 1:ESlateVisibility Enum=SelfHitTestInvisible
     Index=Get bUseBackgroundForAffordabilityFeedback()
    ) -> (ReturnValue:ESlateVisibility Enum)
   )
  ]}]
variables [
 bUseBackgroundForAffordabilityFeedback {type=boolean}
 CanAffordBackgroundColor {type=SlateColor, default {colorUseRule=UseColor_Specified, specifiedColor=<rgba 1 1 1 0>}}
 CanAffordTextColor {type=SlateColor, default {colorUseRule=UseColor_Specified, specifiedColor=<rgba 1 1 1 1>}}
 CanNotAffordBackgroundColor {type=SlateColor, default {colorUseRule=UseColor_Specified, specifiedColor=<rgba 0.5 0 0 0.7>}}
 CanNotAffordTextColor {type=SlateColor, default {colorUseRule=UseColor_Specified, specifiedColor=<rgba 1 0.1686 0.2078 1>}}
]
widgets {
 info {inheritedWidgetCount=3, parentClass=/Script/FortniteUI.FortResourceWidget, rootWidgetClass=&Overlay, widgetCount=8}
 widgets [
  Overlay_2 {widgetClassPath=&Overlay, widget=Overlay_2}
  Border {
   widgetClassPath=&CommonBorder
   parent=Overlay_2
   widget=Border
   properties {
    background {
     bIsDynamicallyLoaded=false
     drawAs=Box
     imageType=NoImage
     mirroring=NoMirror
     resourceName=None
     resourceObject=/FortUILibrary/Styles/Border/Materials/M_Border_HUD_VGradient
     tiling=NoTile
     imageSize=<32 32>
     margin=<ltrb 0.49 0.49 0.49 0.49>
     outlineSettings {
      bUseBrushTransparency=false
      roundingType=HalfHeightRadius
      width=0
      color {colorUseRule=UseColor_Specified, specifiedColor=<rgba 0 0 0 0>}
      cornerRadii=<0 0 0 0>
     }
     tintColor {colorUseRule=UseColor_Specified, specifiedColor=<rgba 0.0052 0.0052 0.0056 0>}
     uVRegion {bIsValid=false, max=<0 0>, min=<0 0>}
    }
    brushColor=<rgba 0.0386 0 1 1>
    padding=<ltrb 0 0 0 0>
   }
   slot {class=OverlaySlot, properties {horizontalAlignment=HAlign_Fill, verticalAlignment=VAlign_Fill}}
  }
  Background {
   widgetClassPath=&Border
   bIsVariable=true
   parent=Overlay_2
   widget=Background
   properties {
    visibility=SelfHitTestInvisible
    background {
     bIsDynamicallyLoaded=false
     drawAs=Box
     imageType=NoImage
     mirroring=NoMirror
     resourceName=None
     resourceObject=null
     tiling=NoTile
     imageSize=<32 32>
     margin=<ltrb 0 0 0 0>
     outlineSettings {
      bUseBrushTransparency=false
      roundingType=HalfHeightRadius
      width=0
      color {colorUseRule=UseColor_Specified, specifiedColor=<rgba 0 0 0 0>}
      cornerRadii=<0 0 0 0>
     }
     tintColor {colorUseRule=UseColor_Specified, specifiedColor=<rgba 0.5 0 0 0.7>}
     uVRegion {bIsValid=false, max=<0 0>, min=<0 0>}
    }
    padding=<ltrb 0 0 0 0>
   }
   slot {class=OverlaySlot, properties {horizontalAlignment=HAlign_Fill, verticalAlignment=VAlign_Fill}}
  }
  SizeBoxContainer {
   widgetClassPath=&SizeBox
   parent=Overlay_2
   widget=SizeBoxContainer
   properties {bOverride_HeightOverride=true, bOverride_WidthOverride=true, heightOverride=80, maxDesiredWidth=64, minDesiredWidth=64, widthOverride=64}
   slot {class=OverlaySlot, properties {horizontalAlignment=HAlign_Fill, verticalAlignment=VAlign_Fill}}
  }
  VerticalBox_0 {widgetClassPath=&VerticalBox, parent=SizeBoxContainer, widget=VerticalBox_0, slot {class=SizeBoxSlot}}
  LazyImage_ResourceIcon {
   widgetClassPath=&FortLazyImage
   bInherited=true
   parent=VerticalBox_0
   widget=LazyImage_ResourceIcon
   properties {brush {
     bIsDynamicallyLoaded=false
     drawAs=Image
     imageType=NoImage
     mirroring=NoMirror
     resourceName=None
     resourceObject=null
     tiling=NoTile
     imageSize=<54 54>
     margin=<ltrb 0 0 0 0>
     outlineSettings {
      bUseBrushTransparency=false
      roundingType=HalfHeightRadius
      width=0
      color {colorUseRule=UseColor_Specified, specifiedColor=<rgba 0 0 0 0>}
      cornerRadii=<0 0 0 0>
     }
     tintColor {colorUseRule=UseColor_Specified, specifiedColor=<rgba 1 1 1 1>}
     uVRegion {bIsValid=false, max=<0 0>, min=<0 0>}
    }}
   slot {class=VerticalBoxSlot, properties {horizontalAlignment=HAlign_Center, verticalAlignment=VAlign_Center}}
  }
  NumericText_ResourceCost {
   widgetClassPath=/Script/CommonUI.CommonNumericTextBlock
   bInherited=true
   parent=VerticalBox_0
   widget=NumericText_ResourceCost
   properties {
    bSimpleTextMode=true
    style=/FortUILibrary/Styles/Typography/Foundation/TextStyle-Base-XS-B.TextStyle-Base-XS-B_C
    text=0
    font {
     bForceMonospaced=false
     bMaterialIsStencil=false
     fontMaterial=null
     fontObject=&NotoSans
     letterSpacing=0
     monospacedWidth=1
     size=16
     skewAmount=0
     typefaceFontName=NotoSans-Bold
     outlineSettings {
      bApplyOutlineToDropShadows=false
      bMiteredCorners=false
      bSeparateFillAlpha=false
      outlineMaterial=null
      outlineSize=0
      outlineColor=<rgba 0 0 0 1>
     }
    }
    formattingSpecification {
     alwaysSign=false
     maximumFractionalDigits=0
     maximumIntegralDigits=324
     minimumFractionalDigits=0
     minimumIntegralDigits=1
     roundingMode=HalfFromZero
     useGrouping=false
    }
    shadowOffset=<0 0>
   }
   slot {class=VerticalBoxSlot, properties {horizontalAlignment=HAlign_Center, padding=<ltrb 0 -3 0 0>}}
  }
  NumericText_ResourceCount {widgetClassPath=&CommonNumericTextBlock, bInherited=true, widget=null}
 ]
}
