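# Creative-mode UI widget that shows an actor's item cost as resource counts in a horizontal box inside a translucent border.
# SetItemCost finds the specified ItemCostComponent class on the actor and passes its cost map to SetItemCostInternal with interpolation settings.
# It collapses itself if the actor or component is invalid.
# CreateResourceCountWidget supplies the native ItemCostWidgetBase parent with a ResourceCountWidget per requested resource type.
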WidgetBlueprint /Game/UI/Creative/ItemCostWidget
graphs [
 CreateResourceCountWidget {nodes [
   n1 = CreateResourceCountWidget!() -> (then, bInUseBackgroundForAffordabilityFeedback:bool, ResourceType:EFortResourceType Enum=Wood) {is=title K2Node_FunctionEntry}
   > n2 = Create Widget!(
    execute:Exec
    Class:User Widget Class Reference=&/ResourceCountWidget.ResourceCountWidget_C
    OwningPlayer:Player Controller Object Reference=GetOwningPlayer()
    bUseBackgroundForAffordabilityFeedback:bool=@n1
    ItemDefinition:Fort Resource Item Definition Object Reference
    ResourceType:EFortResourceType Enum=@n1.ResourceType
   ) -> (then:Exec, ReturnValue:Resource Count Widget Object Reference)
   > Return Node!(execute, ReturnValue:Fort Resource Widget Object Reference=@n2)
  ]}
 EventGraph {links [], nodes []}
 SetItemCost {nodes [
   n1 = SetItemCost!() -> (then, Actor:Actor Object Reference, SpecifiedItemCostComponent:Item Cost Component Class Reference) {is=title K2Node_FunctionEntry}
   > n2 = Is Valid(InputObject=@n1)
   n3 = GetComponentByClass!(self:Actor Object Reference=@n1, ComponentClass:Actor Component Class Reference=@n1.SpecifiedItemCostComponent) -> (ReturnValue:Item Cost Component Object Reference)
   > n4 = Is Valid(InputObject=@n3)
   SetVisibility(execute=[@n4.Is Not Valid, @n2.Is Not Valid], InVisibility=Collapsed)
   SetItemCostInternal!(
    execute=@n4.Is Valid
    self:Item Cost Widget Base Object Reference
    Costs:Map of EFortResourceType Enums to Integers=GetCost!(self:Item Cost Component Object Reference=@n3) -> (Cost:Map of EFortResourceType Enums to Integers=Wood)
    MaximumInterpolationDuration:float=3
    MinimumChangeRate:float=1
    OutroOffset:float=0
    bForceResetInterpolate:bool=false
   ) -> (then)
  ]}
]
variables []
widgets {
 info {inheritedWidgetCount=2, parentClass=/Script/FortniteUI.ItemCostWidgetBase, rootWidgetClass=&Border, widgetCount=2}
 widgets [
  RarityBorder {
   widgetClassPath=&Border
   bInherited=true
   bIsVariable=true
   widget=RarityBorder
   properties {brushColor=<rgba 0 0 0 0.5>, padding=<ltrb 16 5 16 5>}
  }
  ResourceWidgetsContainer {
   widgetClassPath=&HorizontalBox
   bInherited=true
   bIsVariable=true
   parent=RarityBorder
   widget=ResourceWidgetsContainer
   slot {class=BorderSlot, properties {padding=<ltrb 16 5 16 5>}}
  }
 ]
}
